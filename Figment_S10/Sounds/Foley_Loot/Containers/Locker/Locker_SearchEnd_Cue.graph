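# Sound cue played, by its name, when searching a locker container ends, layering a "no more loot" UI sound with locker foley.
# The foley picks one of four equally weighted search waves, fades it out with a short volume envelope, and varies pitch and volume slightly.
# A mixer combines both, with the UI layer slightly lower.
# It uses prop-search attenuation, the searching foley sound class and a concurrency cap that stops the farthest instance.
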SoundCue /Figment_S10/Sounds/Foley_Loot/Containers/Locker/Locker_SearchEnd_Cue
root=Mixer_12
nodes [
 {
  id=Enveloper_14
  children [Random_0]
  properties {pitchCurve {editorCurveData {keys [{time=0}]}}, volumeCurve {editorCurveData {keys [<at 0 1>,<at 0.64 0>]}}}
 }
 {id=Mixer_12, children [WavePlayer_43, Modulator_15], properties {inputVolume [0.8,1]}}
 {id=Modulator_15, children [Enveloper_14], properties {pitchMax=1.07, pitchMin=0.92, volumeMax=1.07, volumeMin=0.92}}
 {id=Random_0, children [WavePlayer_8, WavePlayer_9, WavePlayer_10, WavePlayer_11], properties {weights [1,1,1,1]}}
 {id=WavePlayer_10, properties {soundWaveAssetPtr=&/Locker_Search04}}
 {id=WavePlayer_11, properties {soundWaveAssetPtr=&/Locker_Search05}}
 {id=WavePlayer_43, properties {soundWaveAssetPtr=/Game/Sounds/Fort_UI_Sounds/Gameplay/Fort_UI_NoMoreLoot}}
 {id=WavePlayer_8, properties {soundWaveAssetPtr=&/Locker_Search01}}
 {id=WavePlayer_9, properties {soundWaveAssetPtr=&/Locker_Search02}}
]
properties {
 attenuationSettings=/Game/Sounds/Attenuation/Fort_Prop_Search_Att
 bOverrideConcurrency=true
 maxDistance=1300
 soundClassObject=/Figment_S10/Sounds/FortSoundClassesAndModes/Fort_Searching_Foley
 volumeMultiplier=0.7
 concurrencyOverrides {maxCount=4, resolutionRule=StopFarthestThenPreventNew}
}
