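# Sound cue for searching a locker loot container, routed to the searching foley sound class with prop-search attenuation.
# It mixes three randomized layers: an immediate locker search with slight pitch and volume variation, plus two half-volume follow-ups.
# Short randomized delays bring in a second locker search sound, then a generic gear rummage slightly later.
# A concurrency override uses PreventNew, so new plays are refused once the limit is hit.
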SoundCue /Figment_S10/Sounds/Foley_Loot/Containers/Locker/Locker_Search_Cue
root=Mixer_12
nodes [
 {id=Delay_1, children [Random_13], properties {delayMax=0.9, delayMin=0.8}}
 {id=Delay_2, children [Random_16], properties {delayMax=0.8, delayMin=0.7}}
 {id=Mixer_12, children [Modulator_18,Delay_2,Delay_1], properties {inputVolume [1,0.5,0.5]}}
 {id=Modulator_18, children [Random_4], properties {pitchMax=1.07, pitchMin=0.92, volumeMax=1.07, volumeMin=0.92}}
 {id=Random_13, children [WavePlayer_56, WavePlayer_57, WavePlayer_59], properties {weights [1,1,1]}}
 {id=Random_16, children [WavePlayer_6,WavePlayer_7], properties {weights [1,1]}}
 {id=Random_4, children [WavePlayer_4,WavePlayer_5], properties {weights [1,1]}}
 {id=WavePlayer_4, properties {soundWaveAssetPtr=&/Locker_Search01}}
 {id=WavePlayer_5, properties {soundWaveAssetPtr=&/Locker_Search02}}
 {id=WavePlayer_56, properties {soundWaveAssetPtr=/Game/Sounds/Foley_Loot/Containers/Misc/Prop_Gear_Search_05}}
 {id=WavePlayer_57, properties {soundWaveAssetPtr=/Game/Sounds/Foley_Loot/Containers/Misc/Prop_Gear_Search_06}}
 {id=WavePlayer_59, properties {soundWaveAssetPtr=/Game/Sounds/Foley_Loot/Containers/Misc/Prop_Gear_Search_01}}
 {id=WavePlayer_6, properties {soundWaveAssetPtr=&/Locker_Search04}}
 {id=WavePlayer_7, properties {soundWaveAssetPtr=&/Locker_Search05}}
]
properties {
 attenuationSettings=/Game/Sounds/Attenuation/Fort_Prop_Search_Att
 bOverrideConcurrency=true
 maxDistance=1300
 soundClassObject=/Figment_S10/Sounds/FortSoundClassesAndModes/Fort_Searching_Foley
 volumeMultiplier=1.2
 concurrencyOverrides {resolutionRule=PreventNew}
}
